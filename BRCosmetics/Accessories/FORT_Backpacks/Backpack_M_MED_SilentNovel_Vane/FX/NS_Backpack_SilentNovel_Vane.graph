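# Niagara system for the SilentNovel Vane backpack cosmetic that throws short-lived mesh debris in instantaneous bursts.
# Crumbs and Chunks emitters launch particles from a shape, then apply gravity, drag and a grow-then-shrink curve.
# User.ElimCount indexes arrays for spawn arc and rotation, plus Crumbs count.
# Particles.Rand picks palette or time-of-day color and mesh visibility, while secondary black, enlarged, camera-offset attributes suggest outlines.
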NiagaraSystem /BRCosmetics/Accessories/FORT_Backpacks/Backpack_M_MED_SilentNovel_Vane/FX/NS_Backpack_SilentNovel_Vane
dependencies {
 usedDataInterfaces [
  &NiagaraDataInterfaceArrayInt32
  &NiagaraDataInterfaceArrayColor
  &NiagaraDataInterfaceArrayFloat
  &NiagaraDataInterfaceArrayFloat3
  &NiagaraDataInterfaceCurve
 ]
 usedDynamicInputs []
 usedModules [
  &self:SystemSpawnScript.SystemScriptSource.SystemScriptGraph.NiagaraNodeAssignment_0.SetVariables_#1
  &SystemState
 ]
}
emitters {
 Chunks {
  emitterUpdateScript [
   &EmitterState()
   &SpawnBurst_Instantaneous(
    Spawn Count=&RandomRangeInt(Minimum=2, Maximum=3)
    Use Spawn Probability=false
    UseLoopCountLimit=false
   )
  ]
  particleSpawnScript [
   /SetVariables(Particles.Rand=1) {isSetParameters=true, module=SetVariables_#2}
   &InitializeParticle(
    Color=&LerpLinearColors(
     StartColor=&SelectLinearColorFromArray(
      Array Sampling Mode=::NewEnumerator0
      Direct Array Index=?
      Color Selection Array=DataInterface() {propertyValues {ColorData [
         <rgba 1 0.0437 0.0685 1>
         <rgba 0 0.4621 0.9387 1>
         <rgba 0.007 0.5972 0.2016 1>
         <rgba 1 1 0.8308 1>
         <rgba 0.6445 0.314 0.0931 1>
         <rgba 1 1 0 1>
        ]}}
      Random Seed=0
     )
     EndColor=&FNTimeOfDayColor(Night Color=<rgba 1 0.3229 0.3229 1>, Morning Color=<rgba 1 1 1 1>, Day Color=<rgba 1 1 1 1>, Evening Color=<rgba 1 1 1 1>)
     LerpFactor=&MakeFloatFromInt(Int=$Particles.Rand)
    )
    Lifetime=&Multiply_Float(A=&RandomRangeFloat(Minimum=0.4, Maximum=1), B=$System.CrumbsLifeMult)
    Mass Max=2
    Mass Mode=::NewEnumerator3
    Mesh Scale Mode=::NewEnumerator3
    Mesh Uniform Scale=&Multiply_Float(A=2.5, B=$System.CrumbsScale)
    Sprite Size Max=<3 14>
    Sprite Size Min=<2 6>
    Sprite Size Mode=::NewEnumerator2
    UsePositionOffset=false
    Write Material Random=true
   )
   &AddVelocity(
    Module.Use Falloff Distance=false
    Override Seed=false
    Velocity Mode=::NewEnumerator1
    Velocity Speed=&RandomRangeFloat(Minimum=100, Maximum=150)
   ) {module=AddVelocity001}
   &AddVelocity(Velocity=<0 0 1>, Velocity Speed Scale=100) {module=AddVelocity002}
   &AddVelocity(Velocity=<0 -1 0>, Velocity Speed Scale=200)
   &ShapeLocation(
    Handle Radius=5
    Large Radius=10
    MaskSpawnGroup=false
    Offset=<0 -5 0>
    Offset Mode=::NewEnumerator0
    Override Seed=false
    Rotation Angle Type:ENiagara_AngleInput=::NewEnumerator0
    Rotation Mode=::NewEnumerator2
    Shape Primitive=::NewEnumerator3
    U Distribution:NiagaraFloat=&SelectFloatFromArray(Direct Array Index=$User.ElimCount, Float Selection Array=DataInterface() {propertyValues {FloatData [0.4,0.8,0.6,0.4]}})
    UseOwnerScale=true
    Yaw / Pitch / Roll=&SelectVectorFromArray(
     Direct Array Index=$User.ElimCount
     Vector Selection Array=DataInterface() {propertyValues {FloatData [<0 30 90>, <0 110 90>, <0 80 90>, <0 30 90>]}}
    )
   )
   &InitialMeshOrientation(Use Rotation Vector=false)
   /SetVariables(Particles.VisibilityTag=&SelectIntFromArray(Direct Array Index=$Particles.Rand, Int Selection Array=DataInterface() {propertyValues {IntData [0,1]}})) {isSetParameters=true, module=SetVariables_#3}
   &CameraOffset(Camera Offset Amount=10)
  ]
  particleUpdateScript [
   &ParticleState()
   &ScaleMeshSize(Scale Factor=&VectorFromFloat(Value=&FloatFromCurve(FloatCurve=DataInterface() {propertyValues {Curve {
         defaultValue=1
         keys [
          <at 0 0 RCIM_Cubic>
          {arriveTangentWeight=0.0916, interpMode=RCIM_Cubic, leaveTangentWeight=0.281, tangentMode=RCTM_User, time=0.1229, value=1}
          {arriveTangentWeight=0.1257, interpMode=RCIM_Cubic, leaveTangentWeight=0.1667, tangentMode=RCTM_User, time=0.5, value=1}
          {interpMode=RCIM_Cubic, tangentMode=RCTM_User, time=1, value=0}
         ]
        }}})))
   &GravityForce(Gravity=&MakeVector(X=0, Y=0, Z=$System.CrumbsGravity))
   &Drag(Drag=&RandomRangeFloat(Minimum=0.8, Maximum=1.2), Use Linear Drag=true, Use Rotational Drag=false)
   &SolveForcesAndVelocity(Clamp Velocity=false, Limit Acceleration=false)
   /SetVariables(Particles.Scale2:Vector3f=&Multiply_VectorByFloat(Vector=$Particles.Scale, Float=1.15)) {isSetParameters=true, module=SetVariables_#4}
   /SetVariables(Particles.CameraOffset2=-1) {isSetParameters=true, module=SetVariables_#5}
   /SetVariables(Particles.Color2=<rgba 0 0 0 1>) {isSetParameters=true, module=SetVariables_#6}
  ]
  rendererClasses [&NiagaraMeshRendererProperties]
  renderers [
   {rendererClass=&NiagaraMeshRendererProperties, rendererIndex=0}
   {rendererClass=&NiagaraMeshRendererProperties, rendererIndex=1}
   {rendererClass=&NiagaraMeshRendererProperties, rendererIndex=2}
   {rendererClass=&NiagaraMeshRendererProperties, rendererIndex=3}
  ]
 }
 Crumbs {
  emitterUpdateScript [
   &EmitterState()
   &SpawnBurst_Instantaneous(
    Spawn Count=&SelectIntFromArray(Direct Array Index=$User.ElimCount, Int Selection Array=DataInterface() {propertyValues {IntData [35,25,35,45]}})
    Use Spawn Probability=false
    UseLoopCountLimit=false
   )
  ]
  particleSpawnScript [
   /SetVariables(Particles.Rand:NiagaraInt32=&RandomRangeInt(Minimum=0, Maximum=1)) {isSetParameters=true, module=SetVariables_#2}
   &InitializeParticle(
    Color=&LerpLinearColors(
     StartColor=&SelectLinearColorFromArray(
      Array Sampling Mode=::NewEnumerator0
      Direct Array Index=?
      Color Selection Array=DataInterface() {propertyValues {ColorData [
         <rgba 1 0.0437 0.0685 1>
         <rgba 0 0.4621 0.9387 1>
         <rgba 0.007 0.5972 0.2016 1>
         <rgba 1 1 0.8308 1>
         <rgba 0.6445 0.314 0.0931 1>
         <rgba 1 1 0 1>
        ]}}
      Random Seed=0
     )
     EndColor=&FNTimeOfDayColor(Night Color=<rgba 1 0.3229 0.3229 1>, Morning Color=<rgba 1 1 1 1>, Day Color=<rgba 1 1 1 1>, Evening Color=<rgba 1 1 1 1>)
     LerpFactor=&MakeFloatFromInt(Int=$Particles.Rand)
    )
    Lifetime=&Multiply_Float(A=&RandomRangeFloat(Minimum=0.25, Maximum=0.9), B=$System.CrumbsLifeMult)
    Mass Max=2
    Mass Mode=::NewEnumerator3
    Mesh Scale Mode=::NewEnumerator3
    Mesh Uniform Scale=&Multiply_Float(
     A=&Lerp_Float(A=1.5, B=1, Alpha=&MakeFloatFromInt(Int=$Particles.Rand))
     B=$System.CrumbsScale
    )
    Sprite Size Max=<3 14>
    Sprite Size Min=<2 6>
    Sprite Size Mode=::NewEnumerator2
    UsePositionOffset=false
    Write Material Random=true
   )
   &AddVelocity(
    Module.Use Falloff Distance=false
    Override Seed=false
    Velocity Mode=::NewEnumerator1
    Velocity Speed=&RandomRangeFloat(Minimum=100, Maximum=150)
   ) {module=AddVelocity001}
   &AddVelocity(Velocity=<0 0 1>, Velocity Speed Scale=100) {module=AddVelocity002}
   &AddVelocity(Velocity=<0 -1 0>, Velocity Speed Scale=100)
   &ShapeLocation(
    Handle Radius=5
    Large Radius=10
    MaskSpawnGroup=false
    Offset=<0 -5 0>
    Offset Mode=::NewEnumerator0
    Override Seed=false
    Rotation Angle Type:ENiagara_AngleInput=::NewEnumerator0
    Rotation Mode=::NewEnumerator2
    Shape Primitive=::NewEnumerator3
    U Distribution:NiagaraFloat=&SelectFloatFromArray(Direct Array Index=$User.ElimCount, Float Selection Array=DataInterface() {propertyValues {FloatData [0.4,0.8,0.6,0.4]}})
    UseOwnerScale=true
    Yaw / Pitch / Roll=&SelectVectorFromArray(
     Direct Array Index=$User.ElimCount
     Vector Selection Array=DataInterface() {propertyValues {FloatData [<0 30 90>, <0 110 90>, <0 80 90>, <0 30 90>]}}
    )
   )
   &InitialMeshOrientation(Use Rotation Vector=false)
   /SetVariables(Particles.VisibilityTag=&SelectIntFromArray(Direct Array Index=$Particles.Rand, Int Selection Array=DataInterface() {propertyValues {IntData [0,1]}})) {isSetParameters=true, module=SetVariables_#3}
   &CameraOffset(Camera Offset Amount=10)
  ]
  particleUpdateScript [
   &ParticleState()
   &ScaleMeshSize(Scale Factor=&VectorFromFloat(Value=&FloatFromCurve(FloatCurve=DataInterface() {propertyValues {Curve {
         defaultValue=1
         keys [
          <at 0 0 RCIM_Cubic>
          {arriveTangentWeight=0.0916, interpMode=RCIM_Cubic, leaveTangentWeight=0.281, tangentMode=RCTM_User, time=0.1229, value=1}
          {arriveTangentWeight=0.1257, interpMode=RCIM_Cubic, leaveTangentWeight=0.1667, tangentMode=RCTM_User, time=0.5, value=1}
          {interpMode=RCIM_Cubic, tangentMode=RCTM_User, time=1, value=0}
         ]
        }}})))
   &GravityForce(Gravity=&MakeVector(X=0, Y=0, Z=$System.CrumbsGravity))
   &Drag(Drag=&RandomRangeFloat(Minimum=0.8, Maximum=1.2), Use Linear Drag=true, Use Rotational Drag=false)
   &SolveForcesAndVelocity(Clamp Velocity=false, Limit Acceleration=false)
   /SetVariables(Particles.Scale2:Vector3f=&Multiply_VectorByFloat(Vector=$Particles.Scale, Float=1.35)) {isSetParameters=true, module=SetVariables_#4}
   /SetVariables(Particles.CameraOffset2=-1) {isSetParameters=true, module=SetVariables_#5}
   /SetVariables(Particles.Color2=<rgba 0 0 0 1>) {isSetParameters=true, module=SetVariables_#6}
  ]
  rendererClasses [&NiagaraMeshRendererProperties]
  renderers [
   {rendererClass=&NiagaraMeshRendererProperties, rendererIndex=0}
   {rendererClass=&NiagaraMeshRendererProperties, rendererIndex=1}
   {rendererClass=&NiagaraMeshRendererProperties, rendererIndex=2}
   {rendererClass=&NiagaraMeshRendererProperties, rendererIndex=3}
  ]
 }
}
summary {
 systemName=NS_Backpack_SilentNovel_Vane
 emitters [
  {emitterName=Crumbs, rendererClasses [&NiagaraMeshRendererProperties]}
  {emitterName=Chunks, rendererClasses [&NiagaraMeshRendererProperties]}
 ]
 userVariables [User.ElimCount {type=NiagaraInt32}]
}
