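# Anim Blueprint for the Vane backpack cosmetic, which changes its look as the player eliminates enemies.
# Elimination anim input events advance a capped stage; UpdateStages toggles "Hide Element" scalars on mesh and overlay dynamic materials.
# The final stage starts a timed reset, while frontend display cycles stages on a looping timer.
# It also derives pack rotation and translation from spine bend, and layers additive transforms on the pack bone.
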AnimBlueprint /BRCosmetics/Accessories/FORT_Backpacks/Backpack_M_MED_SilentNovel_Vane/Meshes/M_MED_SilentNovel_Vane_Pack_AnimBP
target_skeleton=&/M_MED_SilentNovel_Vane_Pack_Skeleton
graphs [
 AnimGraph {nodes [
   n1 = Input Pose()
   n2 = "Transform (Modify) Bone - Bone: pack"!(
    ComponentPose="Transform (Modify) Bone - Bone: pack"!(ComponentPose=Local To Component(LocalPose=@n1), Alpha:float, Translation="0,0,0", Rotation="0,0,0", Scale) -> (Pose) {rotationMode=BMM_Additive, translationMode=BMM_Additive, boneToModify {boneName=pack}}
    bAlphaBoolEnabled:bool
    Translation="0,-1,0"
    Rotation="0,0,0"
    Scale
   ) -> (Pose) {
    alphaInputType=Bool
    rotationMode=BMM_Additive
    translationMode=BMM_Additive
    alphaBoolBlend {blendInTime=0.2, blendOutTime=0.2}
    boneToModify {boneName=pack}
    translation {y=-1}
   }
   Output Pose(Result=Component To Local(ComponentPose=@n2)) {layerGroup=None}
  ]}
 EventGraph {nodes [
   n1 = Test() {is=title K2Node_CustomEvent}
   > UpdateStages()
   > n2 = Increment Int(Value=Get EliminatedStage() -> (EliminatedStage:int=0))
   > Set EliminatedStage(EliminatedStage:int="Clamp (Integer)"(Value=@n2, Min=0, Max=4) -> (ReturnValue=0)) -> (Output_Get:int=0)
   > Branch(Condition="integer >= integer"(A=Get EliminatedStage() -> (EliminatedStage=0), B=4))
   > Set EliminatedStage(EliminatedStage=0) -> (Output_Get=0)
   n3 = Event BlueprintUpdateAnimation()
   > n4 = Sequence()
   Branch(execute=@n4.then_1, Condition=Get ResetTimerActive() -> (ResetTimerActive:bool))
   > Set TimeSinceMaxStage(TimeSinceMaxStage:double=float + float(A=@n3.DeltaTimeX, B=Get TimeSinceMaxStage() -> (TimeSinceMaxStage:double))) -> (Output_Get:double)
   > Branch(Condition="float >= float"(A=Get TimeSinceMaxStage(), B=45))
   > Set EliminatedStage(EliminatedStage=0) -> (Output_Get=0)
   > Set ResetTimerActive(ResetTimerActive:bool=false) -> (Output_Get:bool)
   > Set TimeSinceMaxStage(TimeSinceMaxStage=0)
   > UpdateStages()
   n5 = GetOwnerSkeletalMeshComponent()
   n6 = BreakRotator(InRot=BreakTransform(InTransform=GetDeltaTransformFromRefPose(self=@n5, BoneName=spine_05, BaseName=spine_04)).Rotation)
   Set BackPack Rotation(execute=@n4.then_0, BackPack Rotation:Rotator=MakeRotator(Roll=MapRangeClamped(Value=@n6.Yaw, InRangeA=0, InRangeB=-30, OutRangeB=20))) -> (Output_Get:Rotator)
   > Set BackPack Translation(BackPack Translation:Vector=MakeVector(Y=MapRangeClamped(Value=@n6.Yaw, InRangeA=0, InRangeB=-30, OutRangeB=-5))) -> (Output_Get:Vector)
   n7 = Event BlueprintProcessAnimInputEvent()
   > Branch(Condition=IsNamed(self=@n7.AnimInputEvent, InEventName=aie.EnemyPlayerEliminated))
   > n8 = Increment Int(Value=Get EliminatedStage() -> (EliminatedStage=0))
   > Set EliminatedStage(EliminatedStage="Clamp (Integer)"(Value=@n8, Min=Get EliminatedStage() -> (EliminatedStage=0), Max=Get MaxEliminatedStage() -> (MaxEliminatedStage:int=0)) -> (ReturnValue=0)) -> (Output_Get=0)
   > UpdateStages()
   Event OnCharacterCustomizationLoaded()
   > n9 = GetOrCreateDynamicOverlayMaterial(MeshComponent=GetPartSkeletalMeshComponent(), MaterialName=None)
   > Set OverlayMat(OverlayMat:Material Instance Dynamic Object Reference=@n9) -> (Output_Get:Material Instance Dynamic Object Reference)
   Event BlueprintBeginPlay() {is=title K2Node_Event}
   > Parent: BlueprintBeginPlay() {is=title K2Node_CallParentFunction}
   > Branch(Condition=IsCosmeticObjectDisplayedInFrontendContext!(CosmeticObject:Object Reference) -> (ReturnValue:bool))
   > Set Timer by Event(Delegate=@n1, Time=3, bLooping=true)
   > Test() {is=title K2Node_CallFunction}
  ]}
 UpdateStages {nodes [
   UpdateStages() {is=title K2Node_FunctionEntry}
   > n1 = Switch on Int!(execute:Exec, Selection:int=Get EliminatedStage() -> (EliminatedStage=0)) -> (Default:Exec, 0:Exec, 1:Exec, 2:Exec, 3:Exec, 4:Exec)
   n2 = Get Dynamic Material Instance(execute=@n1.0, SkeletalMeshComponent=GetPartSkeletalMeshComponent())
   n3 = Get OverlayMat() -> (OverlayMat:Material Instance Dynamic Object Reference)
   > SetScalarParameterValue(self=[@n3, @n2], ParameterName=Hide Element 01, Value=0)
   > SetScalarParameterValue(self=[@n3, @n2], ParameterName=Hide Element 02, Value=1)
   > SetScalarParameterValue(self=[@n3, @n2], ParameterName=Hide Element 03, Value=1)
   > SetScalarParameterValue(self=[@n2, @n3], ParameterName=Hide Element 04, Value=1)
   n4 = Get Dynamic Material Instance(execute=@n1.1, SkeletalMeshComponent=GetPartSkeletalMeshComponent())
   n5 = Get OverlayMat()
   > SetScalarParameterValue(self=[@n5, @n4], ParameterName=Hide Element 01, Value=1)
   > SetScalarParameterValue(self=[@n4, @n5], ParameterName=Hide Element 02, Value=0)
   > SetScalarParameterValue(self=[@n4, @n5], ParameterName=Hide Element 03, Value=1)
   > SetScalarParameterValue(self=[@n4, @n5], ParameterName=Hide Element 04, Value=1)
   n6 = Get Dynamic Material Instance(execute=@n1.2, SkeletalMeshComponent=GetPartSkeletalMeshComponent())
   n7 = Get OverlayMat()
   > SetScalarParameterValue(self=[@n7, @n6], ParameterName=Hide Element 01, Value=1)
   > SetScalarParameterValue(self=[@n7, @n6], ParameterName=Hide Element 02, Value=1)
   > SetScalarParameterValue(self=[@n7, @n6], ParameterName=Hide Element 03, Value=0)
   > SetScalarParameterValue(self=[@n7, @n6], ParameterName=Hide Element 04, Value=1)
   n8 = Get Dynamic Material Instance(execute=@n1.3, SkeletalMeshComponent=GetPartSkeletalMeshComponent())
   n9 = Get OverlayMat()
   > SetScalarParameterValue(self=[@n9, @n8], ParameterName=Hide Element 01, Value=1)
   > SetScalarParameterValue(self=[@n8, @n9], ParameterName=Hide Element 02, Value=1)
   > SetScalarParameterValue(self=[@n8, @n9], ParameterName=Hide Element 03, Value=1)
   > SetScalarParameterValue(self=[@n8, @n9], ParameterName=Hide Element 04, Value=0)
   > Set ResetTimerActive(ResetTimerActive=true)
   > Set TimeSinceMaxStage(TimeSinceMaxStage=0)
  ]}
]
variables [
 BackPack Rotation {type=Rotator}
 BackPack Translation {type=Vector}
 EliminatedStage {type=integer}
 MaxEliminatedStage {type=integer, default=3}
 OverlayMat {type=MaterialInstanceDynamic}
 ResetTimerActive {type=boolean}
 TimeSinceMaxStage {type=number}
]
